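# Translucent UI material that draws a row of round pips for a rotator, with Options setting the count and SelectedIndex marking the current choice.
# UVs tile via Frac, are corrected for widget pixel aspect and UserScale, then become soft-edged circles through a radial SmoothStep driven by Step and Softness.
# Ceil-based band masks isolate the selection, feeding a highlight mask and an opacity term that keeps other pips at half strength.
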Material /FortUILibrary/UIKit/PipList/M_UI_RotatorPips
graph {
 properties {
  bAutomaticallySetUsageInEditor=false
  bUsedWithStaticMesh=false
  blendMode=BLEND_Translucent
  materialDomain=MD_UI
  usedShadingModels=Default Lit
 }
 nodes [
  n1 = TextureCoordinate() {coordinateIndex=4}
  n2 = Add(A=NamedRerouteUsage()) {constB=1.5}
  n3 = NamedRerouteUsage()
  n4 = Multiply(A=Divide(A=Add(A=@n2) {constB=-1}, B=@n3)) {constB=-1}
  n5 = ComponentMask(None=Ceil(None=Add(A=@n1, B=@n4))) {r=true}
  n6 = Add(A=@n1, B=Multiply(A=Divide(A=@n2, B=@n3)) {constB=-1})
  n7 = OneMinus(None=ComponentMask(None=Ceil(None=@n6)) {r=true})
  n8 = NamedRerouteUsage()
  n9 = NamedRerouteUsage()
  n10 = MaterialFunctionCall(Float2=NamedRerouteUsage()) {materialFunction=&BreakOutFloat2Components}
  n11 = Multiply(A=Saturate(None=DotProduct(A=@n5, B=@n7)), B=OneMinus(None=SmoothStep(Min=@n8, Max=@n9, Value=@n10.G)))
  Multiply(
   A=Max(A=NamedRerouteDeclaration(None=@n11) {name=SelectedPipsMask, nodeColor=<rgba 1 0 0.8952 1>}, B=NamedRerouteUsage())
   B=NamedRerouteUsage()
  )
  n12 = MaterialFunctionCall() {materialFunction=&GetUserInterfaceUV}
  n13 = AppendVector(A=Constant() {r=1}, B=Divide(A=ComponentMask(None=@n12.Pixel Size) {g=true}, B=ComponentMask(None=@n12.Pixel Size) {r=true})) {desc=Widget Vertical Aspect Ratio}
  n14 = NamedRerouteUsage()
  n15 = AppendVector(A=Constant() {r=1}, B=Divide(A=Constant() {r=0.9}, B=Divide(A=@n14) {constB=1.1})) {desc=Texture Vertical Aspect Ratio}
  NamedRerouteDeclaration(None=Divide(A=Divide(A=@n13, B=@n15), B=ScalarParameter() {defaultValue=0.75, parameterName=UserScale})) {name=AspectRatio, nodeColor=<rgba 1 0.5391 0 1>}
  NamedRerouteDeclaration(None=Add(A=ScalarParameter() {defaultValue=5, group=Options, parameterName=Options})) {name=Options, nodeColor=<rgba 0 0.7187 1 1>}
  n16 = ScalarParameter() {defaultValue=0.9, parameterName=Step}
  NamedRerouteDeclaration(None=Subtract(A=@n16, B=ScalarParameter() {defaultValue=0.1, parameterName=Softness}) {constB=0.1}) {name=MinStep, nodeColor=<rgba 0 1 0.6953 1>}
  n17 = Multiply(A=TextureCoordinate() {coordinateIndex=4}, B=AppendVector(A=NamedRerouteUsage(), B=Constant() {r=1}))
  n18 = Multiply(A=Add(A=Frac(None=@n17)) {constB=-0.5}) {constB=2}
  n19 = NamedRerouteUsage()
  n20 = Max(A=Abs(None=Multiply(A=@n18, B=@n19))) {constB=0}
  NamedRerouteDeclaration(None=@n20) {name=Frac-ed TextureCoordinates, nodeColor=<rgba 1 0 0.1875 1>}
  n21 = SmoothStep(Min=NamedRerouteUsage(), Max=NamedRerouteUsage(), Value=Length(None=@n20) {is=class MaterialExpressionLength}) {constMin=0.8}
  NamedRerouteDeclaration(None=Saturate(None=OneMinus(None=@n21))) {name=Pips, nodeColor=<rgba 0 0.1094 1 1>}
  NamedRerouteDeclaration(None=ScalarParameter() {group=Options, parameterName=SelectedIndex}) {name=Selected Index, nodeColor=<rgba 1 0 0.6562 1>}
  n22 = TextureCoordinate() {coordinateIndex=4}
  n23 = Add(A=NamedRerouteUsage()) {constB=2}
  n24 = NamedRerouteUsage()
  n25 = Multiply(A=Divide(A=Add(A=@n23) {constB=-2}, B=@n24)) {constB=-1}
  n26 = ComponentMask(None=Ceil(None=Add(A=@n22, B=@n25))) {r=true}
  n27 = Add(A=@n22, B=Multiply(A=Divide(A=@n23, B=@n24)) {constB=-1})
  n28 = OneMinus(None=ComponentMask(None=Ceil(None=@n27)) {r=true})
  n29 = NamedRerouteUsage()
  n30 = Multiply(
   A=Saturate(None=DotProduct(A=@n26, B=@n28))
   B=Step(
    Y=MaterialFunctionCall(Float2=@n29) {materialFunction=&BreakOutFloat2Components}.G
    X=NamedRerouteUsage()
   ) {constX=0.65, constY=1}
  )
  NamedRerouteDeclaration(None=Saturate(None=Add(A=@n30) {constB=0.5})) {name=Opacity, nodeColor=<rgba 0.4297 1 0 1>}
  NamedRerouteDeclaration(None=@n18) {nodeColor=<rgba 0.2656 0 1 1>}
  NamedRerouteDeclaration(None=@n16) {name=MaxStep, nodeColor=<rgba 0 1 0.6953 1>}
 ]
}
