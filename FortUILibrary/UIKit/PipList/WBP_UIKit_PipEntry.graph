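# UIKit pip list entry widget: a small material-driven indicator that animates between inactive and active states.
# A Spacer inside the Pip Border sets its size, so the short Anim_Active animation widens the pip while raising the material's Progress Opacity.
# SetActive plays that animation forward to activate or in reverse to deactivate.
# IsImmediate starts playback at the animation's end time, snapping to the target state without a visible transition.
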WidgetBlueprint /FortUILibrary/UIKit/PipList/WBP_UIKit_PipEntry
animations [Anim_Active {
  binding_count=2
  length_seconds=0.15
  object=&self:Anim_Active
  sequence {
   bindings [
    Pip {
     id=#1
     possessed_class=Border
     tracks [{
       type=MovieSceneWidgetMaterialTrack
       sections [{
         type=MovieSceneParameterSection
         end_frame=9
         start_frame=0
         channels [Progress Opacity {type=MovieSceneScriptingFloatChannel, keys [[0,0],[9,1]]}]
        }]
      }]
    }
    PipSpacer {
     id=#2
     possessed_class=Spacer
     tracks [{
       type=MovieSceneDoubleVectorTrack
       display_name=Size
       sections [{
         type=MovieSceneDoubleVectorSection
         end_frame=9
         start_frame=0
         channels [
          Vector.X {type=MovieSceneScriptingDoubleChannel, keys [[0,8],[9,16]]}
          Vector.Y {type=MovieSceneScriptingDoubleChannel, const [8,1]}
         ]
        }]
      }]
    }
   ]
   display_rate {numerator=60}
  }
 }]
graphs [EventGraph {nodes [
   n1 = SetActive!() -> (OutputDelegate, then, Active:bool=~, IsImmediate:bool) {is=title K2Node_CustomEvent}
   n2 = Get Anim_Active() -> (Anim_Active:Widget Animation Object Reference)
   > PlayAnimation(
    InAnimation=Get Anim_Active()
    StartAtTime=Select(Option 0:float=0, Option 1:float=GetEndTime(self=@n2), Index=@n1.IsImmediate) -> (ReturnValue:float)
    PlayMode=Select(Option 0:EUMGSequencePlayMode Enum=Reverse, Option 1:EUMGSequencePlayMode Enum, Index=@n1.Active) -> (ReturnValue:EUMGSequencePlayMode Enum)
   )
  ]}]
variables []
widgets {
 info {inheritedWidgetCount=0, parentClass=&UserWidget, rootWidgetClass=&Border, widgetCount=2}
 widgets [
  Pip {
   widgetClassPath=&Border
   widget=Pip
   properties {
    horizontalAlignment=HAlign_Left
    visibility=SelfHitTestInvisible
    background {
     bIsDynamicallyLoaded=false
     drawAs=Image
     imageType=FullColor
     mirroring=NoMirror
     resourceName=None
     resourceObject=&/MI_UIKit_Pip_Entry
     tiling=NoTile
     imageSize=<8 8>
     margin=<ltrb 0.5 0.5 0.5 0.5>
     outlineSettings {
      bUseBrushTransparency=false
      roundingType=HalfHeightRadius
      width=0
      color {colorUseRule=UseColor_Specified, specifiedColor=<rgba 0 0 0 0>}
      cornerRadii=<0 0 0 0>
     }
     tintColor {colorUseRule=UseColor_Specified, specifiedColor=<rgba 1 1 1 1>}
     uVRegion {bIsValid=false, max=<0 0>, min=<0 0>}
    }
    padding=<ltrb 0 0 0 0>
   }
  }
  PipSpacer {
   widgetClassPath=&Spacer
   parent=Pip
   widget=PipSpacer
   properties {size=<8 8>}
   slot {class=BorderSlot, properties {horizontalAlignment=HAlign_Left, padding=<ltrb 0 0 0 0>}}
  }
 ]
}
